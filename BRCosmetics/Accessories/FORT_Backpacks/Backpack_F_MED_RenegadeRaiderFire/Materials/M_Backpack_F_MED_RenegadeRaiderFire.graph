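# Surface material for the Renegade Raider Fire backpack cosmetic, set up as Default Lit for skeletal meshes and mesh deformers.
# A black-by-default Emissive Color parameter feeds the MF_FX_EmissiveAdjustments material function.
# Three loose scalar constants complete the small graph, which contains no texture samples and shows no links to material outputs.
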Material /BRCosmetics/Accessories/FORT_Backpacks/Backpack_F_MED_RenegadeRaiderFire/Materials/M_Backpack_F_MED_RenegadeRaiderFire
graph {
 properties {bUsedWithMeshDeformer=true, bUsedWithSkeletalMesh=true, pixelDepthOffsetMode=PDOM_Legacy, usedShadingModels=Default Lit}
 nodes [
  Constant()
  Constant()
  Constant() {r=1}
  MaterialFunctionCall(Emissive=VectorParameter() {parameterName=Emissive Color, defaultValue=<rgba 0 0 0 1>}.RGB) {materialFunction=&MF_FX_EmissiveAdjustments}
 ]
}
